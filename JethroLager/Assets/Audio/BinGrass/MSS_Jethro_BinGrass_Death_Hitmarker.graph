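# Mono one-shot MetaSound for the BinGrass boss death hitmarker, routed to the 2D instigator hit-notification sound class with hitmarker attenuation.
# A single wave player fires on play with the sweetener wave, slightly pitch-shifted, feeding the output and finish trigger directly.
# The Base and Sweetner wave arrays, gain inputs and recent-play interface inputs are left unconnected.
# A concurrency override allows one voice, with a short steal release.
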MetaSoundSource /JethroLager/Assets/Audio/BinGrass/MSS_Jethro_BinGrass_Death_Hitmarker
inputs [
 ActiveSound.bIsOccluded {type=Bool}
 ActiveSound.MaxDistance {type=Float}
 Gain.Base {type=Float}
 Gain.Sweetner {type=Float}
 RecentCollection.NumRecentPlays {type=Int32}
 RecentCollection.TimeSinceLastPlayed {type=Float}
 UE.Source.OnPlay {type=Trigger}
 Wavs.Base {type=WaveAsset:Array}
 Wavs.Sweetner {type=WaveAsset:Array}
]
interfaces [Fort.ActiveSound:1.0, Fort.ActiveSoundRecentCollection:1.0, UE.OutputFormat.Mono:1.0, UE.Source.OneShot:1.0, UE.Source:1.1]
outputs [
 "UE.OutputFormat.Mono.Audio:0" {type=Audio}
 UE.Source.OneShot.OnFinished {type=Trigger}
]
pages [Default {
  input_defaults {
   ActiveSound.MaxDistance=0
   ActiveSound.bIsOccluded=null
   Gain.Base=0.892
   Gain.Sweetner=1
   RecentCollection.NumRecentPlays=0
   RecentCollection.TimeSinceLastPlayed=10
   UE.Source.OnPlay=null
   Wavs.Base [
    /JethroLager/NPCs/Bosses/BinGrass/Audio/Waves/BinGrass_ArmorHit_01
    &BinGrass_ArmorHit_01
    &BinGrass_ArmorHit_01
    &BinGrass_ArmorHit_01
    &BinGrass_ArmorHit_01
    &BinGrass_ArmorHit_01
    &BinGrass_ArmorHit_01
    &BinGrass_ArmorHit_01
   ]
   Wavs.Sweetner [/JethroLager/NPCs/Bosses/BinGrass/Audio/Waves/BinGrass_HitMarker_Sweetener]
  }
  nodes [WavePlayer_0 = UE.Wave Player:Mono(Play=^UE.Source.OnPlay) {also {defaults {Pitch Shift=0.7, Wave Asset=&BinGrass_HitMarker_Sweetener}}}]
  links [{dst="@UE.Source.OneShot.OnFinished", src=WavePlayer_0.On Finished}, {dst="@UE.OutputFormat.Mono.Audio:0", src=WavePlayer_0.Out Mono}]
 }]
properties {
 assetImportData=&self:AssetImportData
 attenuationSettings=/Game/Athena/Sounds/HitNotifications/Att_HitMarker_Instigate_MS
 bOverrideConcurrency=true
 sampleRate=48000
 soundClassObject=&Fort_2DHitNotifications_Instigate_PSM
 volume=1.2
 concurrencyOverrides {maxCount=1, voiceStealReleaseTime=0.1}
}
