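# Mono one-shot MetaSound source for the BinGrass hitmarker sound, layering randomly chosen waves.
# On play, it fires a pitch-randomized base wave, plus a pitch-randomized sweetener when time since last play passes a threshold comparison from the recent-collection interface.
# A third random layer reads a wave-array variable with no waves visibly assigned, at random gain.
# The mixed output finishes after the base and any triggered sweetener end.
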MetaSoundSource /JethroLager/Assets/Audio/BinGrass/MSS_Jethro_BinGrass_Hitmarker_Lib
inputs [
 Gain.Base {type=Float}
 Gain.Sweetner {type=Float}
 MaxTimeBetweenSweetenedShot {type=Float}
 RecentCollection.NumRecentPlays {type=Int32}
 RecentCollection.TimeSinceLastPlayed {type=Float}
 UE.Source.OnPlay {type=Trigger}
 Wavs.Base {type=WaveAsset:Array}
 Wavs.Sweetner {type=WaveAsset:Array}
]
interfaces [Fort.ActiveSoundRecentCollection:1.0, UE.OutputFormat.Mono:1.0, UE.Source.OneShot:1.0, UE.Source:1.1]
outputs [
 "UE.OutputFormat.Mono.Audio:0" {type=Audio}
 UE.Source.OneShot.OnFinished {type=Trigger}
]
pages [Default {
  input_defaults {
   Gain.Base=1
   Gain.Sweetner=1
   MaxTimeBetweenSweetenedShot=5
   RecentCollection.NumRecentPlays=0
   RecentCollection.TimeSinceLastPlayed=10
   UE.Source.OnPlay=null
   Wavs.Base=null
   Wavs.Sweetner []
  }
  nodes [
   ArrayRandomGet_1 = Array.Random Get:WaveAsset:Array(Next=^UE.Source.OnPlay, In Array=^Wavs.Base) {also {defaults {Weights []}}}
   n1 = UE.RandomFloat(Next=@ArrayRandomGet_1.On Next, Reset=@ArrayRandomGet_1.On Reset)
   WavePlayer_0 = UE.Wave Player:Mono(Wave Asset=@ArrayRandomGet_1.Value, Play=@n1.On Next, Pitch Shift=@n1.Value) {also {version [1,1]}}
   n2 = TriggerCompare.Clamp:Float(B=^MaxTimeBetweenSweetenedShot, A=^RecentCollection.TimeSinceLastPlayed, Compare=^UE.Source.OnPlay) {also {defaults {Type=5}}}
   ArrayRandomGet_0 = Array.Random Get:WaveAsset:Array(In Array=^Wavs.Sweetner, Next=@n2.True) {also {defaults {Weights []}}}
   n3 = UE.RandomFloat(Next=@ArrayRandomGet_0.On Next, Reset=@ArrayRandomGet_0.On Reset)
   WavePlayer_1 = UE.Wave Player:Mono(Wave Asset=@ArrayRandomGet_0.Value, Play=@n3.On Next, Pitch Shift=@n3.Value) {also {version [1,1]}}
   ArrayRandomGet_2 = Array.Random Get:WaveAsset:Array(Next=^UE.Source.OnPlay, In Array=VariableAccessor.WaveAsset:Array(Variable=InitVariable.WaveAsset:Array().Variable).Value)
   WavePlayer_2 = UE.Wave Player:Mono(Play=@ArrayRandomGet_2.On Next, Wave Asset=@ArrayRandomGet_2.Value) {also {version [1,1]}}
   AudioMixerAudioMixerMono3_0 = "AudioMixer.Audio Mixer (Mono, 3)"(
    Gain 1=^Gain.Base
    Gain 0=^Gain.Sweetner
    Gain 2=UE.RandomFloat() {also {defaults {Max=0.9, Min=0.1}}}.Value
    In 1=@WavePlayer_0.Out Mono
    In 0=@WavePlayer_1.Out Mono
    In 2=@WavePlayer_2.Out Mono
   )
   TriggerAccumulatorTriggerAccumulate2_0 = "TriggerAccumulator.Trigger Accumulate (2)"(In 0=@n2.False, In 1=@WavePlayer_0.On Finished) {also {defaults {Auto Reset="<MetaSoundBuilderResult.SUCCEEDED: 0>"}}}
   TriggerAccumulatorTriggerAccumulate2_1 = "TriggerAccumulator.Trigger Accumulate (2)"(In 1=@WavePlayer_0.On Finished, In 0=@WavePlayer_1.On Finished) {also {defaults {Auto Reset="<MetaSoundBuilderResult.SUCCEEDED: 0>"}}}
   TriggerAnyTriggerAny2_0 = "TriggerAny.Trigger Any (2)"(In 1=@TriggerAccumulatorTriggerAccumulate2_0.Out, In 0=@TriggerAccumulatorTriggerAccumulate2_1.Out)
  ]
  links [
   {dst="@UE.OutputFormat.Mono.Audio:0", src=AudioMixerAudioMixerMono3_0.Out}
   {dst="@UE.Source.OneShot.OnFinished", src=TriggerAnyTriggerAny2_0.Out}
  ]
 }]
properties {
 assetImportData=&self:AssetImportData
 attenuationSettings=/Game/Athena/Sounds/HitNotifications/Att_HitMarker_Instigate_MS
 sampleRate=48000
 soundClassObject=&Fort_2DHitNotifications_Instigate_PSM
}
