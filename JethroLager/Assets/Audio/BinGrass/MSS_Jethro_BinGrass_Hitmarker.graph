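# One-shot mono MetaSound source for the BinGrass boss hitmarker, played through the instigator 2D hit-notification sound class.
# It is a thin wrapper that passes trigger, gains and wave arrays for standard body hits and BinGrass sweeteners into MSS_Jethro_BinGrass_Hitmarker_Lib.
# Recent-play count and time since last play also feed that library graph, which supplies the audio and finished trigger.
# A concurrency override caps simultaneous voices.
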MetaSoundSource /JethroLager/Assets/Audio/BinGrass/MSS_Jethro_BinGrass_Hitmarker
inputs [
 ActiveSound.bIsOccluded {type=Bool}
 ActiveSound.MaxDistance {type=Float}
 Gain.Base {type=Float}
 Gain.Sweetner {type=Float}
 RecentCollection.NumRecentPlays {type=Int32}
 RecentCollection.TimeSinceLastPlayed {type=Float}
 UE.Source.OnPlay {type=Trigger}
 Wavs.Base {type=WaveAsset:Array}
 Wavs.Sweetner {type=WaveAsset:Array}
]
interfaces [Fort.ActiveSound:1.0, Fort.ActiveSoundRecentCollection:1.0, UE.OutputFormat.Mono:1.0, UE.Source.OneShot:1.0, UE.Source:1.1]
outputs [
 "UE.OutputFormat.Mono.Audio:0" {type=Audio}
 UE.Source.OneShot.OnFinished {type=Trigger}
]
pages [Default {
  input_defaults {
   ActiveSound.MaxDistance=0
   ActiveSound.bIsOccluded=null
   Gain.Base=1
   Gain.Sweetner=1
   RecentCollection.NumRecentPlays=0
   RecentCollection.TimeSinceLastPlayed=10
   UE.Source.OnPlay=null
   Wavs.Base [
    /Game/Athena/Sounds/HitNotifications/HitMarker_Body_01
    /Game/Athena/Sounds/HitNotifications/HitMarker_Body_02
    /Game/Athena/Sounds/HitNotifications/HitMarker_Body_03
   ]
   Wavs.Sweetner [
    /JethroLager/NPCs/Bosses/BinGrass/Audio/Waves/BinGrass_HitMarker_Sweetener_01
    /JethroLager/NPCs/Bosses/BinGrass/Audio/Waves/BinGrass_HitMarker_Sweetener_02
   ]
  }
  nodes [MSS_Jethro_BinGrass_Hitmarker_Lib_0 = &/MSS_Jethro_BinGrass_Hitmarker_Lib(
    Gain.Base=^Gain.Base
    Gain.Sweetner=^Gain.Sweetner
    RecentCollection.NumRecentPlays=^RecentCollection.NumRecentPlays
    RecentCollection.TimeSinceLastPlayed=^RecentCollection.TimeSinceLastPlayed
    UE.Source.OnPlay=^UE.Source.OnPlay
    Wavs.Base=^Wavs.Base
    Wavs.Sweetner=^Wavs.Sweetner
   )]
  links [
   {dst="@UE.OutputFormat.Mono.Audio:0", src=MSS_Jethro_BinGrass_Hitmarker_Lib_0.UE.OutputFormat.Mono.Audio:0}
   {dst="@UE.Source.OneShot.OnFinished", src=MSS_Jethro_BinGrass_Hitmarker_Lib_0.UE.Source.OneShot.OnFinished}
  ]
 }]
properties {
 assetImportData=&self:AssetImportData
 attenuationSettings=/Game/Athena/Sounds/HitNotifications/Att_HitMarker_Instigate_MS
 bOverrideConcurrency=true
 sampleRate=48000
 soundClassObject=&Fort_2DHitNotifications_Instigate_PSM
 volume=1.2
 concurrencyOverrides {maxCount=4, voiceStealReleaseTime=0.1}
}
